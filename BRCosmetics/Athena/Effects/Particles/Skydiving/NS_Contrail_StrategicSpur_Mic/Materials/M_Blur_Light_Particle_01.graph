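# Translucent unlit Niagara sprite material for a blurred light spark particle in the StrategicSpur_Mic skydiving contrail.
# Particle color RGB feeds an emissive adjustment function, while particle alpha is multiplied by a motion-blur spark texture and camera depth fade.
# A Sprite_Capsule world position offset function takes particle size as its minimum scale factor.
# A separate procedural sine-and-sphere-mask shape and a dynamic parameter node feed no other shown node.
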Material /BRCosmetics/Athena/Effects/Particles/Skydiving/NS_Contrail_StrategicSpur_Mic/Materials/M_Blur_Light_Particle_01
graph {
 properties {
  bAutomaticallySetUsageInEditor=false
  bUsedWithNiagaraSprites=true
  blendMode=BLEND_Translucent
  pixelDepthOffsetMode=PDOM_Legacy
  shadingModel=MSM_Unlit
  usedShadingModels=Unlit
 }
 nodes [
  DynamicParameter() {defaultValue=<rgba 100 1 1 1>, paramNames [SphereMask_Hardness, Param2, Param3, Param4]}
  n1 = ParticleColor()
  MaterialFunctionCall(Emissive=@n1.RGB) {materialFunction=&MF_FX_EmissiveAdjustments}
  n2 = StaticBool()
  MaterialFunctionCall!(
   Particle Position
   Normalized Particle Direction
   Capsule Diameter
   Capsule Pivot Alignment
   Scale Along Velocity Vector
   ---------Scale Over Velocity Parameters---------
   Particle Speed
   Minimum Speed
   Max Speed
   Minimum Scale Factor=MaterialFunctionCall(Float2=ParticleSize()) {materialFunction=&BreakOutFloat2Components}.G
   Maximum Scale Factor
   ---------- Advanced -----------
   Cap Size Multiplier=ScalarParameter() {defaultValue=0.75, parameterName=Cap Size Multiplier, sliderMax=1}
   Use Camera Offset=@n2
   Use Advanced Features=@n2
   World Position to Negate
  ) {materialFunction=/Engine/Functions/Engine_MaterialFunctions02/WorldPositionOffset/Sprite_Capsule}
  Multiply(
   A=Multiply(
    A=@n1.A
    B=TextureSample() {texture=/BRCosmetics/Athena/Effects/Particles/Skydiving/NS_Contrail_StrategicSpur_Mic/Textures/T_Motion_Blur_Spark_01}.R
   )
   B=MaterialFunctionCall(Fade Length=Constant() {r=60}, Fade Offset=Constant() {r=30}) {materialFunction=&CameraDepthFade}.Result
  )
  n3 = TextureCoordinate()
  n4 = Power(Base=Sine(None=ComponentMask(None=@n3) {r=true}) {period=2}) {constExponent=5}
  n5 = Sine(None=ComponentMask(None=@n3) {g=true}) {period=2}
  n6 = Constant() {r=0.5}
  Multiply(A=Saturate(None=Multiply(A=@n4, B=@n5)), B=SphereMask(A=TextureCoordinate(), B=@n6, Radius=@n6) {hardnessPercent=0})
 ]
}
