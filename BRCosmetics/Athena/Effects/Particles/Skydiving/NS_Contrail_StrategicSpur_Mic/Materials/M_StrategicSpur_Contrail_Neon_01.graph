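# Unlit, two-sided Niagara mesh-particle material for the neon tubes of the StrategicSpur skydiving contrail.
# Rounded sine bands scroll along V, faster when Dive_Texture_Speed is raised, driving both a three-color blend and local offsets.
# World position offset adds SplineThicken, camera offset and particle-driven tube offsets scaled along bounding-box height.
# Noise varies emissive brightness; a Glow Mesh switch uses flat Color_1 instead, while drag and panner branches appear unconnected.
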Material /BRCosmetics/Athena/Effects/Particles/Skydiving/NS_Contrail_StrategicSpur_Mic/Materials/M_StrategicSpur_Contrail_Neon_01
graph {
 properties {bUsedWithNiagaraMeshParticles=true, pixelDepthOffsetMode=PDOM_Legacy, shadingModel=MSM_Unlit, twoSided=true, usedShadingModels=Unlit}
 nodes [
  n1 = Normalize(VectorInput=Subtract(A=WorldPosition().XYZ, B=CameraPositionWS()))
  n2 = ScalarParameter() {parameterName=CameraOffset}
  n3 = Multiply(A=Time(), B=ScalarParameter() {defaultValue=0.75, parameterName=Colors_Speed})
  n4 = Round(None=DynamicParameter() {defaultValue=<rgba 0 1 1 1>, paramNames [Tubes_Dive_Offset, Dive_Texture_Speed, Tubes_Offset, Param4]}.Dive_Texture_Speed)
  n5 = Add(A=TextureCoordinate(), B=LinearInterpolate(A=@n3, B=Multiply(A=@n3) {constB=1.8}, Alpha=Saturate(None=@n4)))
  n6 = Sine(None=ComponentMask(None=@n5) {g=true})
  n7 = Saturate(None=Round(None=Power(Base=@n6) {constExponent=1}))
  n8 = ComponentMask(None=Add(A=@n5) {constB=-0.33}) {g=true}
  n9 = Round(None=Power(Base=Sine(None=@n8)) {constExponent=1})
  n10 = Saturate(None=Add(A=Multiply(A=@n7) {constB=0.5}, B=Saturate(None=@n9)))
  n11 = MaterialFunctionCall() {materialFunction=&BoundingBoxBased_0-1_UVW}
  n12 = Constant3Vector() {constant=<rgba -17 0 0 0>}
  n13 = DynamicParameter() {defaultValue=<rgba 0 1 1 1>, paramNames [Tubes_Dive_Offset, Dive_Texture_Speed, Tubes_Offset, Param4]}
  n14 = Multiply(A=VertexColor(), B=Constant3Vector() {constant=<rgba -21 0 0 0>})
  n15 = Add(
   A=Multiply(A=Multiply(A=@n10, B=@n11.B), B=Multiply(A=@n12, B=@n13.Tubes_Offset)) {constB=44}
   B=Multiply(A=@n13.Tubes_Dive_Offset, B=Multiply(A=@n11.B, B=@n14))
  )
  n16 = ScalarParameter() {defaultValue=1.8, parameterName=SplineThicken_Width}
  Add(
   A=Add(A=Multiply(A=@n1, B=@n2), B=Transform(None=@n15) {transformSourceType=TRANSFORMSOURCE_Local})
   B=MaterialFunctionCall(WidthBase=@n16, WidthTip=@n16) {materialFunction=&SplineThicken}.WorldPositionOffset
  )
  n17 = VectorParameter() {parameterName=Color_1, defaultValue=<rgba 0.9734 0 1 1>}
  n18 = MaterialFunctionCall() {materialFunction=&BoundingBoxBased_0-1_UVW}
  n19 = Multiply(
   A=Saturate(None=Add(A=@n18.B) {constB=0.5})
   B=MaterialFunctionCall(
    A=VectorParameter() {parameterName=Color_2, defaultValue=<rgba 0.1072 0 1.3 1>}.RGB
    B=@n17.RGB
    C=VectorParameter() {parameterName=Color_3, defaultValue=<rgba 0 1 0.5776 1>}.RGB
    Alpha=@n10
   ) {materialFunction=&Lerp_3Color}.Result
  )
  n20 = Multiply(A=@n19) {constB=1.7}
  n21 = Multiply(A=ComponentMask(None=@n18.RGB) {b=true, r=true}) {constB=0.3}
  n22 = TextureSample(UVs=@n21) {samplerType=SAMPLERTYPE_LinearColor, texture=/Game/Global/Textures/Default/Noise/T_Noise01}
  n23 = Saturate(None=@n22.R)
  MaterialFunctionCall(Emissive=StaticSwitchParameter(True=@n17.RGB, False=LinearInterpolate(A=@n19, B=@n20, Alpha=@n23)) {group=Glow Mesh, parameterName=Glow Mesh?}) {materialFunction=&MF_FX_EmissiveAdjustments}
  n24 = DynamicParameter() {parameterIndex=1}
  n25 = MaterialFunctionCall(X=@n24.Param1, Y=@n24.Param1, Z=@n24.Param1) {materialFunction=&MakeFloat3}
  n26 = ParticlePositionWS()
  n27 = ScalarParameter() {defaultValue=100, parameterName=Particle_WPO_Drag_Amount}
  n28 = Power(Base=MaterialFunctionCall() {materialFunction=&BoundingBoxBased_0-1_UVW}.B)
  Multiply(A=Clamp(None=Subtract(A=@n25.Result, B=@n26), Min=Multiply(A=@n27) {constB=-1}, Max=@n27), B=Multiply(A=@n24.Param4, B=Saturate(None=@n28)))
  n29 = Ceil(None=Multiply(A=Time(), B=ScalarParameter() {defaultValue=0.75, parameterName=Colors_Speed}))
  n30 = Round(None=DynamicParameter() {defaultValue=<rgba 0 1 1 1>, paramNames [Tubes_Dive_Offset, Dive_Texture_Speed, Tubes_Offset, Param4]}.Dive_Texture_Speed)
  Panner(Coordinate=@n21, Time=LinearInterpolate(A=@n29, B=Multiply(A=@n29) {constB=1.8}, Alpha=Saturate(None=@n30))) {speedY=-0.3}
  Power(Base=@n22.R)
  n31 = Sine(None=ComponentMask(None=TextureCoordinate()) {r=true}) {period=2}
  n32 = Multiply(A=ScalarParameter() {defaultValue=1, group=Glow Mesh, parameterName=Glow_Opacity}, B=Saturate(None=Power(Base=@n31)))
  StaticSwitchParameter(True=DepthFade(Opacity=@n32, FadeDistance=ScalarParameter() {defaultValue=100, parameterName=Glow_Depth_Fade}), False=@n32) {parameterName=UseDepthFade?}
 ]
}
