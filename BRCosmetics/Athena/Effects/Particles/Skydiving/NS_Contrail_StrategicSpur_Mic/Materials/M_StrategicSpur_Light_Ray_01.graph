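# Translucent, unlit, two-sided mesh-particle material for light rays in the StrategicSpur skydiving contrail.
# Layers soft and hard light from panning noise, masked by vertical UV gradients and view-angle falloff that softens silhouette edges.
# Niagara dynamic parameters offset the UVs and switch panning to double speed for diving.
# The hard mask blends two colors tinted by particle color; summed layers, boosted at one end, drive depth-faded opacity.
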Material /BRCosmetics/Athena/Effects/Particles/Skydiving/NS_Contrail_StrategicSpur_Mic/Materials/M_StrategicSpur_Light_Ray_01
graph {
 properties {
  bUsedWithNiagaraMeshParticles=true
  blendMode=BLEND_Translucent
  pixelDepthOffsetMode=PDOM_Legacy
  shadingModel=MSM_Unlit
  twoSided=true
  usedShadingModels=Unlit
 }
 nodes [
  n1 = ParticleColor()
  n2 = ComponentMask(None=TextureCoordinate()) {g=true}
  n3 = OneMinus(None=@n2)
  n4 = DynamicParameter() {paramNames [Param1, Dive_Texture_Speed, Texture_Offset1, Texture_Offset2]}
  n5 = AppendVector(A=@n4.Texture_Offset1, B=@n4.Texture_Offset1)
  n6 = Add(A=TextureCoordinate(), B=@n5)
  n7 = VectorParameter() {parameterName=MainTexScale, defaultValue=<rgba 1 0.3 0 0.3>}
  n8 = ComponentMask(None=@n7.RGB) {g=true, r=true}
  n9 = Time()
  n10 = Multiply(A=@n9) {constB=2}
  n11 = Saturate(None=Round(None=@n4.Dive_Texture_Speed))
  n12 = TextureSampleParameter2D(UVs=Panner(Coordinate=Multiply(A=@n6, B=@n8), Time=LinearInterpolate(A=@n9, B=@n10, Alpha=@n11), Speed=AppendVector(A=@n7.B, B=@n7.B)) {speedY=0.3}) {
   parameterName=MainTex
   samplerType=SAMPLERTYPE_LinearColor
   texture=/BRCosmetics/Athena/Effects/Particles/Items/Flopper/FlopperUpdrade/T_RollingTexture_G
  }
  n13 = ScalarParameter() {defaultValue=4, group=Soft Light, parameterName=Soft_Lights_Texture_PWR}
  n14 = Multiply(A=Saturate(None=Power(Base=@n3) {constExponent=1}), B=Saturate(None=Power(Base=@n12.R, Exp=@n13) {constExponent=4})) {constA=1}
  n15 = Power(Base=@n2) {constExponent=1.4}
  n16 = Abs(None=DotProduct(A=VertexNormalWS(), B=CameraVectorWS()))
  n17 = ScalarParameter() {defaultValue=1.6, group=Soft Light, parameterName=Soft_Lights_Fresnel}
  n18 = Multiply(A=Multiply(A=@n14, B=Saturate(None=@n15)), B=Saturate(None=Power(Base=@n16, Exp=@n17) {constExponent=1.6}))
  n19 = ScalarParameter() {defaultValue=2, group=Hard Light, parameterName=Hard_Lights_Fresnel}
  n20 = Add(A=@n5, B=TextureCoordinate())
  n21 = VectorParameter() {parameterName=DetailTextureScale, defaultValue=<rgba 2 0.1 0.05 0.2>}
  n22 = ComponentMask(None=@n21.RGB) {g=true, r=true}
  n23 = Time()
  n24 = Multiply(A=@n23) {constB=2}
  n25 = Saturate(None=Round(None=DynamicParameter() {paramNames [Param1, Dive_Texture_Speed, Texture_Offset1, Texture_Offset2]}.Dive_Texture_Speed))
  n26 = TextureSampleParameter2D(UVs=Panner(Coordinate=Multiply(A=@n20, B=@n22), Time=LinearInterpolate(A=@n23, B=@n24, Alpha=@n25), Speed=AppendVector(A=@n21.B, B=@n21.B)) {speedX=0.05, speedY=0.2}) {group=Hard Light, parameterName=Hard_Light_Texutre, texture=/BRCosmetics/Weapons/FORT_Melee/Pickaxe_MadameMoth/FX/T_Fibre_Noise}
  n27 = ComponentMask(None=TextureCoordinate()) {g=true}
  n28 = Power(Base=OneMinus(None=@n27)) {constExponent=1}
  n29 = Multiply(
   A=Power(Base=@n26.R, Exp=ScalarParameter() {defaultValue=1, group=Hard Light, parameterName=Hard_Lights_Texture_PWR}) {constExponent=4}
   B=Multiply(A=Saturate(None=@n27), B=Saturate(None=@n28))
  )
  n30 = ScalarParameter() {defaultValue=1, group=Hard Light, parameterName=Hard_Lights_Texture_Mult}
  n31 = Multiply(A=Saturate(None=Power(Base=@n16, Exp=@n19)), B=Saturate(None=Multiply(A=@n29, B=@n30) {constB=5}))
  n32 = Add(A=@n18, B=Multiply(A=@n31, B=ScalarParameter() {defaultValue=1, group=Hard Light, parameterName=Hard_Lights_Opacity}) {constB=0.1})
  n33 = Power(Base=ComponentMask(None=TextureCoordinate()) {g=true}) {constExponent=6}
  n34 = Multiply(A=Multiply(A=@n32, B=Saturate(None=@n33)), B=ScalarParameter() {defaultValue=15, parameterName=Bottom_Grad_Add_Mult}) {constB=15}
  DepthFade(Opacity=Multiply(A=@n1.A, B=Add(A=@n34, B=@n32)))
  n35 = VectorParameter() {group=Soft Light, parameterName=Color_1, defaultValue=<rgba 2.1871 0 4 1>}
  n36 = VectorParameter() {group=Hard Light, parameterName=Color_2, defaultValue=<rgba 0 2.9743 4 1>}
  n37 = Power(Base=ScalarParameter() {defaultValue=0.2264, parameterName=Color_PWR}, Exp=@n14)
  n38 = ScalarParameter() {defaultValue=1.0672, parameterName=Color_Mult}
  LinearInterpolate(A=@n35.RGB, B=@n36.RGB, Alpha=Saturate(None=Multiply(A=@n37, B=@n38) {constB=0.1}))
  n39 = Saturate(None=@n31)
  MaterialFunctionCall(Emissive=Multiply(A=LinearInterpolate(A=@n35.RGB, B=@n36.RGB, Alpha=@n39), B=@n1.RGB)) {materialFunction=&MF_FX_EmissiveAdjustments}
  Multiply(B=@n18)
  Sine(None=@n2) {period=2}
  Sine(None=@n27) {period=2}
 ]
}
