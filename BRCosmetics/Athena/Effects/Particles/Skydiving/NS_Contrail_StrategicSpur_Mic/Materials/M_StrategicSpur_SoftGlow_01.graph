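# Unlit, two-sided translucent soft-glow material for Niagara mesh particles in the StrategicSpur skydiving contrail.
# A Color parameter supplies the tint, while opacity multiplies a view-facing falloff, a sine-based UV fade toward all edges and an Opacity scalar.
# A local-X offset, weighted by bounding-box height and driven by the Tubes_Offset dynamic parameter, is transformed from local space to bend the mesh.
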Material /BRCosmetics/Athena/Effects/Particles/Skydiving/NS_Contrail_StrategicSpur_Mic/Materials/M_StrategicSpur_SoftGlow_01
graph {
 properties {
  bAutomaticallySetUsageInEditor=false
  bUsedWithNiagaraMeshParticles=true
  blendMode=BLEND_Translucent
  pixelDepthOffsetMode=PDOM_Legacy
  shadingModel=MSM_Unlit
  twoSided=true
  usedShadingModels=Unlit
 }
 nodes [
  n1 = CameraVectorWS()
  n2 = VertexNormalWS()
  n3 = Power(Base=Abs(None=DotProduct(A=@n1, B=@n2))) {constExponent=0.5}
  n4 = TextureCoordinate()
  n5 = Multiply(A=Sine(None=ComponentMask(None=@n4) {r=true}) {period=2}, B=Sine(None=ComponentMask(None=@n4) {g=true}) {period=2})
  Multiply(A=Multiply(A=Saturate(None=@n3), B=Saturate(None=@n5)), B=ScalarParameter() {defaultValue=0.4, parameterName=Opacity})
  n6 = Constant3Vector() {constant=<rgba -15 0 0 0>}
  n7 = MaterialFunctionCall() {materialFunction=&BoundingBoxBased_0-1_UVW}
  Transform(None=Multiply(
    A=Multiply(A=@n6, B=@n7.B)
    B=DynamicParameter() {defaultValue=<rgba 0 1 1 1>, paramNames [Tubes_Dive_Offset, Dive_Texture_Speed, Tubes_Offset, Param4]}.Tubes_Offset
   )) {transformSourceType=TRANSFORMSOURCE_Local}
  VectorParameter() {parameterName=Color, defaultValue=<rgba 0.008 0 1 1>}
 ]
}
