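# Niagara skydiving contrail cosmetic (StrategicSpur_Mic) that attaches neon-tube meshes, light rays, glowing sprite decals and light blobs to the character's skeletal mesh.
# All four emitters partition particles to pick per-bone or socket offsets, orientations and sizes from arrays.
# System dive state, from the diving attributes module, drives blob rate, speed and material parameters.
# Glide lean and front-end modifiers adjust offsets and orientation, while a distance-fraction timeline scales mesh length.
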NiagaraSystem /BRCosmetics/Athena/Effects/Particles/Skydiving/NS_Contrail_StrategicSpur_Mic/NS_StrategicSpur_Mic_01
dependencies {
 usedDataInterfaces [
  &NiagaraDataInterfaceCurve
  &NiagaraDataInterfaceArrayFloat
  &NiagaraDataInterfaceArrayColor
  &NiagaraDataInterfaceArrayFloat3
 ]
 usedDynamicInputs [
  &self:Abs_Z_Axis
  &TransformVector
  &NormalizeVector
  &Multiply_Vector
 ]
 usedModules [
  &self:SystemSpawnScript.SystemScriptSource.SystemScriptGraph.NiagaraNodeAssignment_0.SetVariables_#1
  &SystemState
  /Game/Effects/Niagara/Modules/Skydiving/FN_DivingAttributes/v5/FN_DivingAttributes
  &self:SystemSpawnScript.SystemScriptSource.SystemScriptGraph.NiagaraNodeAssignment_1.SetVariables_#2
  &FN_Integrator
  &self:SystemSpawnScript.SystemScriptSource.SystemScriptGraph.NiagaraNodeAssignment_2.SetVariables_#3
  &self:SystemSpawnScript.SystemScriptSource.SystemScriptGraph.NiagaraNodeAssignment_3.SetVariables_#4
 ]
}
emitters {
 Decals {
  emitterUpdateScript [
   &EmitterState(
    Inactive Response=::NewEnumerator1
    Life Cycle Mode=::NewEnumerator1
    Loop Behavior=::NewEnumerator1
    Loop Duration Mode=::NewEnumerator1
    UseLoopDelay=false
   )
   &SpawnBurst_Instantaneous(Spawn Count=4, Use Spawn Probability=false, UseLoopCountLimit=false) {module=SpawnBurst_Instantaneous001}
  ]
  particleSpawnScript [
   &PartitionParticles(Number Of Partitions=4)
   &InitializeParticle(
    Lifetime=0.75
    Sprite Size Mode=::NewEnumerator3
    Uniform Sprite Size=&SelectFloatFromArray(Direct Array Index=$Particles.PartitionParticles.Partition, Float Selection Array=DataInterface() {propertyValues {FloatData [27,27,47,47]}})
    UsePositionOffset=false
    Write Material Random=true
   ) {module=InitializeParticle001}
   /SetVariables(Particles.FlareColor=<rgba 1.8614 0.2014 2 0.1>, Particles.FlareSize:Vector2f=<43 43>) {isSetParameters=true}
  ]
  particleUpdateScript [
   &ParticleState(Kill Particles When Lifetime Has Elapsed=false, Let Infinitely Lived Particles Die When Emitter Deactivates=true)
   &SkeletalMeshLocation(
    Bone / Socket Index=$Particles.PartitionParticles.Partition
    Bone Sampling Mode=::NewEnumerator4
    "Interpolate Mesh Sampling (Update)"=true
    MaskSpawnGroup=false
    Orientation Sampling=::NewEnumerator1
    Sampled Position Offset=&SelectVectorFromArray(
     Direct Array Index=$Particles.PartitionParticles.Partition
     Vector Selection Array=DataInterface() {propertyValues {FloatData [<6 -4 0>, <-6 4 0>, <-8 3 0>, <8 -3 0>]}}
    )
    Skeletal Mesh=$System.Skeletal Mesh
   )
   &CameraOffset(Camera Offset Amount=15)
   &ScaleColor(
    Scale Alpha=&FN_FrontEndFloatModifier(Gameplay=8, Front End=3)
    Scale RGB=&MakeVectorFromLinearColorRGB(Color=<rgba 1.8943 0 2 1>)
    ScaleA=true
    ScaleRGB=true
   ) {module=ScaleColor003}
  ]
  rendererClasses [&NiagaraSpriteRendererProperties]
  renderers [
   {rendererClass=&NiagaraSpriteRendererProperties}
   {rendererClass=&NiagaraSpriteRendererProperties, rendererIndex=1}
  ]
 }
 LightBlobs {
  emitterUpdateScript [
   &EmitterState()
   &SpawnRate(SpawnRate=&Lerp_Float(A=55, B=75, Alpha=$System.Dive), Use Spawn Probability=false)
  ]
  particleSpawnScript [
   &PartitionParticles(Number Of Partitions=4)
   &InitializeParticle(
    Color=&SelectLinearColorFromArray(
     Array Sampling Mode=::NewEnumerator0
     Direct Array Index=?
     Color Selection Array=DataInterface() {propertyValues {ColorData [<rgba 0 1 0.6705 1>, <rgba 0.0134 0 1 1>, <rgba 0.5572 0 1 1>]}}
     Random Seed=0
    )
    Lifetime=&Lerp_Float(
     A=&SelectFloatFromArray(
      Direct Array Index=$Particles.PartitionParticles.Partition
      Float Selection Array=DataInterface() {propertyValues {FloatData [0.7,0.7,0.25,0.25]}}
     )
     B=0.25
     Alpha=$System.Dive
    )
    Sprite Size Max=&FN_FrontEndVector2Modifier(Front End:Vector2f=<15 90>, Gameplay:Vector2f=<15 90>)
    Sprite Size Min=&FN_FrontEndVector2Modifier(Front End:Vector2f=<9 60>, Gameplay:Vector2f=<9 60>)
    Sprite Size Mode=::NewEnumerator2
    Sprite UV Mode=::NewEnumerator3
    UsePositionOffset=false
    Write Material Random=true
   )
   &SkeletalMeshLocation(
    Bone / Socket Index=$Particles.PartitionParticles.Partition
    Bone Sampling Mode=::NewEnumerator4
    MaskSpawnGroup=true
    Orientation Sampling=::NewEnumerator1
    Sampled Position Offset=&SelectVectorFromArray(
     Direct Array Index=$Particles.PartitionParticles.Partition
     Vector Selection Array=DataInterface() {propertyValues {FloatData [<8 7 0>, <-8 -7 0>, <-8 0 0>, <8 0 0>]}}
    )
    Skeletal Mesh=$System.Skeletal Mesh
   ) {module=SkeletalMeshLocation001}
   &ShapeLocation(
    MaskSpawnGroup=false
    Override Seed=false
    Sphere Radius=&SelectFloatFromArray(Direct Array Index=$Particles.PartitionParticles.Partition, Float Selection Array=DataInterface() {propertyValues {FloatData [12,12,5,5]}})
    UseOwnerScale=true
   )
   &CameraOffset(Camera Offset Amount=32)
   &DynamicMaterialParameters(
    Index 0 Param 1=&UniformAOrBFloat(A=0.14, B=0.4, Distribution Weight=0.156, Evaluation Type=::NewEnumerator0, Recalculate Random Each Loop=true, Override Seed=false)
    Module.Index0Param0Write=true
    Module.Index0Param1Write=true
    Module.Index0Param2Write=true
    Module.Index0Param3Write=true
   )
   &AddVelocity(
    Cone Angle=25
    Cone Axis=&FN_FrontEndVector3Modifier(
     Gameplay=&SelectVectorFromArray(
      Direct Array Index=$Particles.PartitionParticles.Partition
      Vector Selection Array=DataInterface() {propertyValues {FloatData [<0.1 -0.05 1>, <-0.1 -0.05 1>, <0 -0.2 1>, <0 -0.2 1>]}}
     )
     Front End=&Lerp_Vector(
      A=&SelectVectorFromArray(
       Direct Array Index=$Particles.PartitionParticles.Partition
       Vector Selection Array=DataInterface() {propertyValues {FloatData [<0.1 -0.05 1>, <-0.1 -0.05 1>, <0 -0.2 1>, <0 -0.2 1>]}}
      )
      B=&SelectVectorFromArray(
       Direct Array Index=$Particles.PartitionParticles.Partition
       Vector Selection Array=DataInterface() {propertyValues {FloatData [<0.1 -0.6 1>, <-0.1 -0.6 1>, <0 -0.5 1>, <0 -0.5 1>]}}
      )
      Alpha=$User.ParaGlideLeanAlpha
     )
    )
    Module.UseConeFalloff=false
    Module.UseVelDistribution=false
    Override Seed=false
    Velocity Mode=::NewEnumerator2
    Velocity Speed=&Lerp_Float(
     A=&RandomRangeFloat(Minimum=250, Maximum=350)
     B=&RandomRangeFloat(Minimum=500, Maximum=600)
     Alpha=$System.Dive
    )
   ) {module=AddVelocity001}
  ]
  particleUpdateScript [
   &ParticleState()
   &ScaleSpriteSize(
    Scale Sprite Size Mode=::NewEnumerator0
    Uniform Scale Factor=&Lerp_Float(
     A=&SelectFloatFromArray(Direct Array Index=$Particles.PartitionParticles.Partition, Float Selection Array=DataInterface() {propertyValues {FloatData [1,1,0.7,0.7]}})
     B=&SelectFloatFromArray(Direct Array Index=$Particles.PartitionParticles.Partition, Float Selection Array=DataInterface() {propertyValues {FloatData [1,1,0.7,0.7]}})
     Alpha=$System.Dive
    )
   ) {module=ScaleSpriteSize001}
   &ScaleSpriteSize(
    Uniform Curve Scale=0.4
    Uniform Curve Sprite Scale=DataInterface() {propertyValues {Curve {keys [
        {arriveTangent=2.0604, interpMode=RCIM_Cubic, leaveTangent=2.0604, leaveTangentWeight=0.4178, tangentMode=RCTM_User, time=0, value=0.35}
        {arriveTangent=1.3, interpMode=RCIM_Cubic, leaveTangent=1.3, time=0.1, value=1}
        {arriveTangent=-1.1111, arriveTangentWeight=0.1389, interpMode=RCIM_Cubic, leaveTangent=-1.1111, leaveTangentWeight=0.2778, time=0.5, value=1}
        {arriveTangent=-2.6251, interpMode=RCIM_Cubic, leaveTangent=-2.6251, tangentMode=RCTM_User, time=1, value=0}
       ]}}}
   )
   &SolveForcesAndVelocity(Clamp Velocity=false, Limit Acceleration=false)
   &ScaleColor(
    Scale Alpha=&FloatFromCurve(
     FloatCurve=DataInterface() {propertyValues {Curve {keys [<at 0 0 RCIM_Cubic>,<at 0.15 1 RCIM_Cubic>,<at 1 0>]}}}
     Scale Curve=&Lerp_Float(A=0.5, B=0.7, Alpha=$System.Dive)
    )
    ScaleA=true
    ScaleRGB=true
   )
  ]
  rendererClasses [&NiagaraSpriteRendererProperties]
  renderers [{rendererClass=&NiagaraSpriteRendererProperties}]
 }
 LightRay {
  emitterUpdateScript [
   &EmitterState(
    Inactive Response=::NewEnumerator1
    Life Cycle Mode=::NewEnumerator1
    Loop Behavior=::NewEnumerator1
    Loop Duration Mode=::NewEnumerator1
    UseLoopDelay=false
   )
   &SpawnRate(SpawnRate=4, Use Spawn Probability=false) {enabled=false}
   /SetVariables(Emitter.DistanceFractionThreshold:NiagaraBool=&SetBoolByFloatComparison(A=$Emitter.Scalability.DistanceFraction, B=0.9, Comparison Type=::NewEnumerator6)) {isSetParameters=true, module=SetVariables_#5}
   &SpawnBurst_Instantaneous(Spawn Count=4, Use Spawn Probability=false, UseLoopCountLimit=false)
   &Timeline(Play=$Emitter.DistanceFractionThreshold, Play Rate=4, Rewind Rate=4) {module=Timeline001}
  ]
  particleSpawnScript [
   &PartitionParticles(Number Of Partitions=4)
   &InitializeParticle(
    Color Mode=::NewEnumerator0
    Lifetime=&RandomRangeFloat(Minimum=1.7, Maximum=2)
    Mesh Scale Mode=::NewEnumerator1
    Mesh Uniform Scale Max=1
    Mesh Uniform Scale Min=0.8
    UsePositionOffset=false
    Write Material Random=true
   )
   &InitialMeshOrientation(
    Mesh Axis To Orient=&FN_FrontEndVector3Modifier(Gameplay=<1 0 0>, Front End=<1 0 0>)
    Mesh Orientation Mode=::NewEnumerator2
    Orientation Coordinate Space=::Local
    Rotation=&FN_FrontEndVector3Modifier(
     Gameplay=&SelectVectorFromArray(
      Direct Array Index=$Particles.PartitionParticles.Partition
      Vector Selection Array=DataInterface() {propertyValues {FloatData [<0.03 0.01 0.09>, <0.03 -0.01 -0.09>, <0 0 0>, <0 0 0>]}}
     )
     Front End=&SelectVectorFromArray(
      Direct Array Index=$Particles.PartitionParticles.Partition
      Vector Selection Array=DataInterface() {propertyValues {FloatData [<0.05 0.03 0.09>, <0.05 -0.03 -0.09>, <0 0 0>, <0 0 0>]}}
     )
    )
    Rotation Coordinate Space=::NewEnumerator2
    Use Rotation Vector=true
   ) {module=InitialMeshOrientation001}
  ]
  particleUpdateScript [
   &ParticleState(Kill Particles When Lifetime Has Elapsed=false, Let Infinitely Lived Particles Die When Emitter Deactivates=true)
   &UpdateMeshOrientation(
    Facing Direction=&Lerp_Vector(
     A=&SelectVectorFromArray(
      Direct Array Index=$Particles.PartitionParticles.Partition
      Vector Selection Array=DataInterface() {propertyValues {FloatData [<1 0 0>, <1 0 0>]}}
     )
     B=&SelectVectorFromArray(
      Direct Array Index=$Particles.PartitionParticles.Partition
      Vector Selection Array=DataInterface() {propertyValues {FloatData [<1 0 0>, <1 0 0>]}}
     )
     Alpha=$User.ParaGlideLeanAlpha
    )
    Facing Interpolation=false
    Orientation Method=::NewEnumerator1
    Override Quat=false
    TransformFacing=false
    TransformUp=false
    Up Direction:Vector3f=&FN_FrontEndVector3Modifier(Gameplay=<0 0 1>, Front End=&Lerp_Vector(A=<0 0 1>, B=<0 -0.5 1>, Alpha=$User.ParaGlideLeanAlpha))
   ) {module=UpdateMeshOrientation002}
   &Timeline(
    Play Rate=&Lerp_Float(A=1, B=2, Alpha=&ClampFloat(Float=$User.ParaGlideLeanAlpha))
    Timeline End=10000000
   )
   &ScaleMeshSize(Scale Factor=&MakeVector(
     X=1
     Y=1
     Z=&FloatFromCurve(
      FloatCurve=DataInterface() {propertyValues {Curve {keys [
          {arriveTangent=5.7371, interpMode=RCIM_Cubic, leaveTangent=5.7371, tangentMode=RCTM_User, time=0, value=0}
          {arriveTangent=-0.9106, interpMode=RCIM_Cubic, leaveTangent=-0.9106, tangentMode=RCTM_User, time=0.76, value=1.025}
          <at 1 1>
         ]}}}
      CurveIndex=$Emitter.Age
      Scale Curve=$Emitter.Timeline001.NormalizedTime
     )
    ))
   &DynamicMaterialParameters(
    Index 0 Param 2=$System.Dive
    Index 0 Param 3=&RandomRangeFloat(Minimum=0, Maximum=1)
    Index 0 Param 4=&RandomRangeFloat(Minimum=0, Maximum=1)
    Module.Index0Param0Write=true
    Module.Index0Param1Write=true
    Module.Index0Param2Write=true
    Module.Index0Param3Write=true
   ) {module=DynamicMaterialParameters001}
   &SkeletalMeshLocation(
    Bone / Socket Index=$Particles.PartitionParticles.Partition
    Bone Sampling Mode=::NewEnumerator4
    MaskSpawnGroup=false
    Orientation Sampling=::NewEnumerator1
    Sampled Position Offset=&FN_FrontEndVector3Modifier(
     Gameplay=&MakeCustomVectorFromBool(
      Bool=$User.bIsFrontEndPreview
      True Vector=&SelectVectorFromArray(
       Direct Array Index=$Particles.PartitionParticles.Partition
       Vector Selection Array=DataInterface() {propertyValues {FloatData [<8 7 -1>, <-8 -7 1>, <0 0 0>, <0 0 0>]}}
      )
      False Vector=&Lerp_Vector(
       A=&SelectVectorFromArray(
        Direct Array Index=$Particles.PartitionParticles.Partition
        Vector Selection Array=DataInterface() {propertyValues {FloatData [<8 7 -1>, <-8 -7 1>, <0 0 0>, <0 0 0>]}}
       )
       B=&SelectVectorFromArray(
        Direct Array Index=$Particles.PartitionParticles.Partition
        Vector Selection Array=DataInterface() {propertyValues {FloatData [<25 0 0>, <-25 0 0>, <0 0 0>, <0 0 0>]}}
       )
       Alpha=$User.ParaGlideLeanAlpha
      )
     )
     Front End=&Lerp_Vector(
      A=&SelectVectorFromArray(
       Direct Array Index=$Particles.PartitionParticles.Partition
       Vector Selection Array=DataInterface() {propertyValues {FloatData [<8 7 -1>, <-8 -7 1>, <0 0 0>, <0 0 0>]}}
      )
      B=&SelectVectorFromArray(
       Direct Array Index=$Particles.PartitionParticles.Partition
       Vector Selection Array=DataInterface() {propertyValues {FloatData [<25 0 0>, <-25 0 0>, <0 0 0>, <0 0 0>]}}
      )
      Alpha=$User.ParaGlideLeanAlpha
     )
    )
    Skeletal Mesh=$System.Skeletal Mesh
   )
   &ScaleMeshSize(Scale Factor=&FN_FrontEndVector3Modifier(Gameplay=<0.7 0.7 0.8>, Front End=<0.65 0.65 0.75>)) {module=ScaleMeshSize001}
   &ScaleMeshSize(Scale Factor=&SelectVectorFromArray(
     Direct Array Index=$Particles.PartitionParticles.Partition
     Vector Selection Array=DataInterface() {propertyValues {FloatData [<1 1 1>, <1 1 1>, <0.5 0.5 0.4>, <0.5 0.5 0.4>]}}
    )) {module=ScaleMeshSize002}
   &ScaleColor(
    Scale Alpha=&FloatFromCurve(
     FloatCurve=DataInterface() {propertyValues {Curve {keys [<at 0 1 RCIM_Cubic>,<at 0.2 1 RCIM_Cubic>,<at 1 1>]}}}
     Scale Curve=&Lerp_Float(A=0.4, B=0.55, Alpha=$System.Dive)
    )
    Scale RGB=&MakeVectorFromLinearColorRGB(Color=&FNTimeOfDayColor(
      Night Color=<rgba 0.6 0.6 0.6 1>
      Morning Color=<rgba 0.65 0.65 0.65 1>
      Day Color=<rgba 0.9 0.9 0.9 1>
      Evening Color=<rgba 0.65 0.65 0.65 1>
     ))
    ScaleA=true
    ScaleRGB=true
   )
   &ScaleColor(Scale Alpha=&FN_FrontEndFloatModifier(Gameplay=1, Front End=0.6), ScaleA=true, ScaleRGB=true) {module=ScaleColor001}
  ]
  rendererClasses [&NiagaraMeshRendererProperties]
  renderers [{rendererClass=&NiagaraMeshRendererProperties, rendererIndex=0}]
 }
 NeonTubes {
  emitterUpdateScript [
   &EmitterState(
    Inactive Response=::NewEnumerator1
    Life Cycle Mode=::NewEnumerator1
    Loop Behavior=::NewEnumerator1
    Loop Duration Mode=::NewEnumerator1
    UseLoopDelay=false
   )
   /SetVariables(Emitter.DistanceFractionThreshold:NiagaraBool=&SetBoolByFloatComparison(A=$Emitter.Scalability.DistanceFraction, B=0.9, Comparison Type=::NewEnumerator6)) {isSetParameters=true, module=SetVariables_#5}
   &SpawnBurst_Instantaneous(Spawn Count=2, Use Spawn Probability=false, UseLoopCountLimit=false)
   &Timeline(Play=$Emitter.DistanceFractionThreshold, Play Rate=4, Rewind Rate=4) {module=Timeline001}
  ]
  particleSpawnScript [
   &PartitionParticles()
   &InitializeParticle(
    Color Mode=::NewEnumerator0
    Mesh Scale Mode=::NewEnumerator3
    Mesh Uniform Scale=&FN_FrontEndFloatModifier(Gameplay=1, Front End=0.85)
    UsePositionOffset=false
    Write Material Random=true
   )
   &InitialMeshOrientation(
    Mesh Axis To Orient=&FN_FrontEndVector3Modifier(Gameplay=<1 0 0>, Front End=<1 0 0>)
    Mesh Orientation Mode=::NewEnumerator2
    Orientation Coordinate Space=::Local
    Rotation=&FN_FrontEndVector3Modifier(
     Gameplay=&SelectVectorFromArray(
      Direct Array Index=$Particles.PartitionParticles.Partition
      Vector Selection Array=DataInterface() {propertyValues {FloatData [<0 0.01 0.07>, <0 -0.01 -0.07>]}}
     )
     Front End=&SelectVectorFromArray(
      Direct Array Index=$Particles.PartitionParticles.Partition
      Vector Selection Array=DataInterface() {propertyValues {FloatData [<0 0.01 0.09>, <0 -0.01 -0.09>]}}
     )
    )
    Rotation Coordinate Space=::NewEnumerator2
    Use Rotation Vector=true
   )
  ]
  particleUpdateScript [
   &ParticleState(Kill Particles When Lifetime Has Elapsed=false, Let Infinitely Lived Particles Die When Emitter Deactivates=true)
   &UpdateMeshOrientation(
    Facing Direction=&Lerp_Vector(
     A=&SelectVectorFromArray(
      Direct Array Index=$Particles.PartitionParticles.Partition
      Vector Selection Array=DataInterface() {propertyValues {FloatData [<1 0 0>, <1 0 0>]}}
     )
     B=&SelectVectorFromArray(
      Direct Array Index=$Particles.PartitionParticles.Partition
      Vector Selection Array=DataInterface() {propertyValues {FloatData [<1 0 0>, <1 0 0>]}}
     )
     Alpha=$User.ParaGlideLeanAlpha
    )
    Facing Interpolation=false
    Orientation Method=::NewEnumerator1
    Override Quat=false
    TransformFacing=false
    TransformUp=false
    Up Direction:Vector3f=&FN_FrontEndVector3Modifier(Gameplay=<0 0 1>, Front End=&Lerp_Vector(A=<0 0 1>, B=<0 -0.5 1>, Alpha=$User.ParaGlideLeanAlpha))
   ) {module=UpdateMeshOrientation002}
   &ScaleMeshSize(Scale Factor=&MakeVector(
     X=1
     Y=1
     Z=&FloatFromCurve(
      FloatCurve=DataInterface() {propertyValues {Curve {keys [
          {arriveTangent=5.7371, interpMode=RCIM_Cubic, leaveTangent=5.7371, tangentMode=RCTM_User, time=0, value=0}
          {arriveTangent=-0.9106, interpMode=RCIM_Cubic, leaveTangent=-0.9106, tangentMode=RCTM_User, time=0.76, value=1.025}
          <at 1 1>
         ]}}}
      CurveIndex=$Emitter.Age
      Scale Curve=$Emitter.Timeline001.NormalizedTime
     )
    ))
   &SkeletalMeshLocation(
    Bone / Socket Index=$Particles.PartitionParticles.Partition
    Bone Sampling Mode=::NewEnumerator4
    MaskSpawnGroup=false
    Orientation Sampling=::NewEnumerator1
    Sampled Position Offset=&FN_FrontEndVector3Modifier(
     Gameplay=&MakeCustomVectorFromBool(
      Bool=$User.bIsFrontEndPreview
      True Vector=&SelectVectorFromArray(
       Direct Array Index=$Particles.PartitionParticles.Partition
       Vector Selection Array=DataInterface() {propertyValues {FloatData [<0 0 0>, <0 0 0>]}}
      )
      False Vector=&Lerp_Vector(
       A=&SelectVectorFromArray(
        Direct Array Index=$Particles.PartitionParticles.Partition
        Vector Selection Array=DataInterface() {propertyValues {FloatData [<8 7 -1>, <-8 -7 1>]}}
       )
       B=&SelectVectorFromArray(
        Direct Array Index=$Particles.PartitionParticles.Partition
        Vector Selection Array=DataInterface() {propertyValues {FloatData [<25 0 0>, <-25 0 0>]}}
       )
       Alpha=$User.ParaGlideLeanAlpha
      )
     )
     Front End=&Lerp_Vector(
      A=&SelectVectorFromArray(
       Direct Array Index=$Particles.PartitionParticles.Partition
       Vector Selection Array=DataInterface() {propertyValues {FloatData [<8 7 -1>, <-8 -7 1>]}}
      )
      B=&SelectVectorFromArray(
       Direct Array Index=$Particles.PartitionParticles.Partition
       Vector Selection Array=DataInterface() {propertyValues {FloatData [<25 0 0>, <-25 0 0>]}}
      )
      Alpha=$User.ParaGlideLeanAlpha
     )
    )
    Skeletal Mesh=$System.Skeletal Mesh
   )
   &ScaleMeshSize(Scale Factor=<0.2 0.2 0.2>) {module=ScaleMeshSize001}
   &DynamicMaterialParameters(
    Index 0 Param 1=&Multiply_Float(
     A=$System.Dive
     B=&SelectFloatFromArray(Direct Array Index=$Particles.PartitionParticles.Partition, Float Selection Array=DataInterface() {propertyValues {FloatData [-1,1]}})
    )
    Index 0 Param 2=$System.Dive
    Index 0 Param 3=&SelectFloatFromArray(Direct Array Index=$Particles.PartitionParticles.Partition, Float Selection Array=DataInterface() {propertyValues {FloatData [-1,1]}})
    Index 0 Param 4=44
    Module.Index0Param0Write=true
    Module.Index0Param1Write=true
    Module.Index0Param2Write=true
    Module.Index0Param3Write=true
   ) {module=DynamicMaterialParameters001}
  ]
  rendererClasses [&NiagaraMeshRendererProperties]
  renderers [
   {rendererClass=&NiagaraMeshRendererProperties, rendererIndex=0}
   {rendererClass=&NiagaraMeshRendererProperties, rendererIndex=1}
   {rendererClass=&NiagaraMeshRendererProperties, rendererIndex=2}
   {rendererClass=&NiagaraMeshRendererProperties, rendererIndex=3}
  ]
 }
}
summary {
 systemName=NS_StrategicSpur_Mic_01
 emitters [
  {emitterName=LightBlobs, rendererClasses [&NiagaraSpriteRendererProperties]}
  {emitterName=NeonTubes, rendererClasses [&NiagaraMeshRendererProperties]}
  {emitterName=LightRay, rendererClasses [&NiagaraMeshRendererProperties]}
  {emitterName=Decals, rendererClasses [&NiagaraSpriteRendererProperties]}
 ]
 userVariables [
  User.bIsFrontEnd {type=NiagaraBool}
  User.bIsFrontEndPreview {type=NiagaraBool}
  User.ParaGlideLeanAlpha {type=NiagaraFloat, defaultValue=0, description=""}
  User.Skeletal Mesh {
   type=NiagaraDataInterfaceSkeletalMesh
   defaultValue {
    struct=NiagaraExt_VariableValue_DataInterface
    value {dataInterface=&self:NiagaraDataInterfaceSkeletalMesh_0}
   }
  }
 ]
}
